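# Animation Blueprint for the StormTracker sniper weapon, built on the bolt-action sniper skeleton.
# It holds a single pose from the StormTracker sniper fire sequence and feeds it through the DefaultSlot, so montages can play over it.
# The update event in the event graph is empty and there are no variables, so the blueprint adds no logic of its own.
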AnimBlueprint /Game/Animation/Game/MainPlayer/Combat/Guns/StormTrackerSniper/BP_StormTrackerSniper
target_skeleton=/Game/Weapons/FORT_Sniper/Meshes/SK_Sniper_Bolt_Action_Skeleton
graphs [
 AnimGraph {nodes [Output Pose(Result=Slot 'DefaultSlot'(Source=Sequence Player 'StormTracker_Sniper_Fire_CMM_W_Pose'() {sequence=&/StormTracker_Sniper_Fire_CMM_W_Pose})) {layerGroup=None}]}
 EventGraph {nodes [Event BlueprintUpdateAnimation()]}
]
variables []
